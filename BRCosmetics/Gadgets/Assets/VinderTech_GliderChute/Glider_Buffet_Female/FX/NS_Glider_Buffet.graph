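# Niagara effect for the Glider_Buffet_Female glider cosmetic, layering blue-purple magic sprites, sparkles, mesh fragments, ribbon trails, rift meshes, flares and a light.
# User deployment, ForwardDot and front-end preview inputs gate many emitters.
# Skeletal mesh sampling places sparkles and trails, and slides caustic sprites between bone positions.
# GPU wing and tail fragments swirl under spring, point and curl-noise forces, while world fragments spawn by distance travelled.
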NiagaraSystem /BRCosmetics/Gadgets/Assets/VinderTech_GliderChute/Glider_Buffet_Female/FX/NS_Glider_Buffet
dependencies {
 usedDataInterfaces [&NiagaraDataInterfaceCurve]
 usedDynamicInputs [&MakeVector2D, &UniformAOrBFloat]
 usedModules [&SystemState, &Timeline]
}
emitters {
 Caustic_Magic {
  emitterUpdateScript [
   &EmitterState(Life Cycle Mode=::NewEnumerator1, Loop Duration=1, UseLoopDelay=false)
   &SpawnRate(
    SpawnRate=&Multiply_Float(A=4, B=&MakeCustomFloatFromBool(Bool=$User.bisFullyDeployed, True Float=1, False Float=0))
    Use Spawn Probability=false
   )
   &SpawnRate(
    SpawnGroup=1
    SpawnRate=&Multiply_Float(A=4, B=&MakeCustomFloatFromBool(Bool=$User.bisFullyDeployed, True Float=1, False Float=0))
    Use Spawn Probability=false
   ) {module=SpawnRate001}
   &SpawnRate(
    SpawnGroup=2
    SpawnRate=&Multiply_Float(A=4, B=&MakeCustomFloatFromBool(Bool=$User.bisFullyDeployed, True Float=1, False Float=0))
    Use Spawn Probability=false
   ) {module=SpawnRate002}
   &SpawnRate(
    SpawnGroup=3
    SpawnRate=&Multiply_Float(A=4, B=&MakeCustomFloatFromBool(Bool=$User.bisFullyDeployed, True Float=1, False Float=0))
    Use Spawn Probability=false
   ) {module=SpawnRate003}
  ]
  particleSpawnScript [
   &InitializeParticle(
    Color=&RandomRangeLinearColor(Minimum=<rgba 0.421 0.2362 0.4427 1>, Maximum=<rgba 0 0.2228 0.6042 1>)
    Lifetime=2.5
    Sprite Rotation Mode=::NewEnumerator2
    Sprite Size Mode=::NewEnumerator3
    Uniform Sprite Size=30
    UsePositionOffset=true
    Write Material Random=true
   ) {module=InitializeParticle001}
   &DynamicMaterialParameters(
    Index 0 Param 1=&RandomRangeFloat(Minimum=0, Maximum=1)
    Index 0 Param 2=&RandomRangeFloat(Minimum=0, Maximum=1)
    Index 0 Param 3=&RandomRangeFloat(Minimum=0, Maximum=1)
    Index 0 Param 4=1
    Module.Param0WriteEnabled=true
    Module.Param1WriteEnabled=true
    Module.Param2WriteEnabled=true
    Module.Param3WriteEnabled=true
   )
   &CameraOffset(Camera Offset Amount=10)
  ]
  particleUpdateScript [
   &ParticleState()
   &SkeletalMeshLocation(
    Bone Sampling Mode=::NewEnumerator4
    Inherit Velocity=false
    MaskSpawnGroup=false
    Sampled Position Offset=<15 0 -5>
    Sampled Position Offset Coordinate Space:ENiagaraMeshTransforms=::NewEnumerator3
    Skeletal Mesh=$User.SkeletalMesh
   )
   /SetVariables(Particles.SkellyAPosition:NiagaraPosition=$Particles.Position) {isSetParameters=true}
   &SkeletalMeshLocation(
    Bone / Socket Index=&Add_Integer(A=1, B=&SpawnGroup())
    Bone Sampling Mode=::NewEnumerator4
    Inherit Velocity=false
    MaskSpawnGroup=false
    Sampled Position Offset Coordinate Space:ENiagaraMeshTransforms=::NewEnumerator3
    Skeletal Mesh=$User.SkeletalMesh
   ) {module=SkeletalMeshLocation001}
   /SetVariables(Particles.SkellyBPosition:NiagaraPosition=$Particles.Position) {isSetParameters=true}
   /SetVariables(Particles.Position=&Lerp_Vector(
     A=$Particles.SkellyAPosition
     B=$Particles.SkellyBPosition
     Alpha=&Multiply_Float(
      A=$Particles.NormalizedAge
      B=&MakeCustomFloatFromBool(
       Bool=&SetBoolByIntComparison(A=&SpawnGroup(), B=2, Comparison Type=::NewEnumerator2)
       True Float=1
       False Float=1.25
      )
     )
    )) {isSetParameters=true}
   &SpriteRotationRate(Rotation Rate=10)
   &ScaleColor(
    Scale Alpha=&FloatFromCurve(FloatCurve=DataInterface() {propertyValues {Curve {keys [
         {arriveTangent=8.5203, interpMode=RCIM_Cubic, leaveTangent=8.5203, tangentMode=RCTM_User, time=0, value=0}
         <at 0.25 1 RCIM_Cubic>
         {arriveTangent=-4.3538, interpMode=RCIM_Cubic, leaveTangent=-4.3538, tangentMode=RCTM_User, time=1, value=0}
        ]}}})
    ScaleA=true
    ScaleRGB=true
   )
   &ScaleSpriteSize(Uniform Curve Sprite Scale=DataInterface() {propertyValues {Curve {keys [<at 0 1 RCIM_Cubic>,<at 1 3 RCIM_Cubic>]}}})
  ]
  rendererClasses [&NiagaraSpriteRendererProperties]
  renderers [{rendererClass=&NiagaraSpriteRendererProperties}]
 }
 FN_AnimTrail {
  emitterSpawnScript [&FN_RibbonIDOffsetUtility()]
  emitterUpdateScript [
   &EmitterState()
   &FN_SpawnAnimTrail(Enable:NiagaraBool=$User.bisFrontEndPreview, Number of Anim Trails=10)
  ]
  particleSpawnScript [
   &FN_InitializeAnimTrailBoneIndecies()
   &FN_InitializeAnimTrailBoneIndecies(AnimTrail Selection=1, First Bone or Socket Index=2, Second Bone or Socket Index=3) {module=FN_InitializeAnimTrailBoneIndecies001}
   &FN_InitializeAnimTrailBoneIndecies(AnimTrail Selection=2, First Bone or Socket Index=4, Second Bone or Socket Index=5) {module=FN_InitializeAnimTrailBoneIndecies002}
   &FN_InitializeAnimTrailBoneIndecies(AnimTrail Selection=3, First Bone or Socket Index=6, Second Bone or Socket Index=7) {module=FN_InitializeAnimTrailBoneIndecies003}
   &FN_InitializeAnimTrailBoneIndecies(AnimTrail Selection=4, First Bone or Socket Index=8, Second Bone or Socket Index=9) {module=FN_InitializeAnimTrailBoneIndecies004}
   &FN_InitializeAnimTrailBoneIndecies(AnimTrail Selection=5, First Bone or Socket Index=10, Second Bone or Socket Index=11) {module=FN_InitializeAnimTrailBoneIndecies005}
   &FN_InitializeAnimTrailBoneIndecies(AnimTrail Selection=6, First Bone or Socket Index=12, Second Bone or Socket Index=13) {module=FN_InitializeAnimTrailBoneIndecies006}
   &FN_InitializeAnimTrailBoneIndecies(AnimTrail Selection=7, First Bone or Socket Index=14, Second Bone or Socket Index=15) {module=FN_InitializeAnimTrailBoneIndecies007}
   &FN_InitializeAnimTrailBoneIndecies(AnimTrail Selection=8, First Bone or Socket Index=16, Second Bone or Socket Index=17) {module=FN_InitializeAnimTrailBoneIndecies008}
   &FN_InitializeAnimTrailBoneIndecies(AnimTrail Selection=9, First Bone or Socket Index=18, Second Bone or Socket Index=19) {module=FN_InitializeAnimTrailBoneIndecies009}
   &FN_InitializeAnimTrail(Color=<rgba 0 1.9333 4 1>)
   &AddVelocity(Velocity=<0 0 800>)
  ]
  particleUpdateScript [
   &SolveForcesAndVelocity(Clamp Velocity=false, Limit Acceleration=false)
   &FN_UpdateAnimTrail(Locked Particle RibbonWidth=1, SkeletalMesh=$User.SkeletalMesh) {module=FN_UpdateAnimTrail001}
   /SetVariables(Particles.RibbonU0Override=$Particles.NormalizedAge) {isSetParameters=true}
   &ScaleColor(Scale Alpha=$System.Timeline.NormalizedTime, ScaleA=true, ScaleRGB=true) {module=ScaleColor001}
   &ScaleColor(
    Scale Alpha=&FloatFromCurve(
     FloatCurve=DataInterface() {propertyValues {Curve {keys [<at 0 1 RCIM_Cubic>, {arriveTangent=-2.1717, interpMode=RCIM_Cubic, leaveTangent=-2.1717, tangentMode=RCTM_User, time=1, value=0}]}}}
     Scale Curve=3
    )
    ScaleA=true
    ScaleRGB=true
   )
   &DynamicMaterialParameters(Module.Param0WriteEnabled=true, Module.Param1WriteEnabled=true, Module.Param2WriteEnabled=true, Module.Param3WriteEnabled=true)
  ]
  rendererClasses [&NiagaraRibbonRendererProperties]
  renderers [{rendererClass=&NiagaraRibbonRendererProperties}]
 }
 Fragment_MainWings_L {
  simTarget=GPUComputeSim
  emitterUpdateScript [
   &EmitterState()
   &SpawnRate(
    SpawnRate=&Multiply_Float(
     A=50
     B=&MakeCustomFloatFromBool(
      Bool=&Boolean_AndOperation(
       Bool1=$User.bisFullyDeployed
       Bool2=&SetBoolByFloatComparison(A=$User.ForwardDot, B=0.1, Comparison Type=::NewEnumerator6)
      )
      True Float=1
      False Float=0
     )
    )
    Use Spawn Probability=false
   )
  ]
  particleSpawnScript [
   &InitializeParticle(
    Color=&MakeCustomLinearColorFromBool(Bool=&RandomBool(), True Color=<rgba 0.9083 0.1374 1.5 1>, False Color=<rgba 0.1213 0.3179 1.5 1>)
    Lifetime Max=5
    Lifetime Min=0.5
    Lifetime Mode=::NewEnumerator1
    Mesh Scale Mode=::NewEnumerator1
    Mesh Uniform Scale Max=0.75
    Mesh Uniform Scale Min=0.25
    UsePositionOffset=false
    Write Material Random=true
   )
   &InitialMeshOrientation(
    Rotation=&RandomRangeVector(Minimum=<0 0 0>, Maximum=<0 0 1>)
    Use Orientation Vector=false
    Use Rotation Vector=true
   )
   &CylinderLocation(
    Cylinder Height=0
    Cylinder Radius=75
    Hemicircle Y=true
    MaskSpawnGroup=true
    Offset=<-15 15 18>
    Offset Mode:ENiagara_OffsetMode=::NewEnumerator0
    Override Seed=false
    Roll=-0.035
    Rotation Mode:ENiagara_RotationMode=::NewEnumerator2
    UseOwnerScale=true
    Yaw=0.225
   )
   &KillParticlesInVolume(
    Cone Angle=30
    Cone Axis:Vector3f=<1 -0.25 0>
    Cone Length=500
    Invert Volume=true
    Kill Shape=::NewEnumerator4
    Origin Offset:Vector3f=<0 0 20>
   )
   &CameraOffset(Camera Offset Amount=&RandomRangeFloat(Minimum=-15, Maximum=5))
  ]
  particleUpdateScript [
   &ParticleState()
   &PointForce(
    Force Falloff Distance=300
    Force Origin=&Add_Vector(A=/Niagara/DynamicInputs/Helpers/SimulationPosition(), B=<0 0 15>)
    Force Strength=200
    Linear Force Vector:Vector3f=<0 0 0>
    Module.Use Falloff=true
    Module.UseLinearForce=false
   )
   &CurlNoiseForce(Pan Noise=false)
   &AccelerationForce(Acceleration=<0 0 -10>)
   &Drag(Drag=3, Use Linear Drag=true, Use Rotational Drag=false)
   /Niagara/Modules/Update/Forces/SpringForce(
    " Particle Equilibrium Position":NiagaraPosition=&Add_Vector(A=&SimulationPosition(), B=<-15 15 18>)
    Force Strength=0.05
    Spring Tightness=3
   )
   &SolveForcesAndVelocity(Clamp Velocity=false, Limit Acceleration=false)
   &ScaleMeshSize(Scale Factor=&VectorFromFloat(Value=&FloatFromCurve(FloatCurve=DataInterface() {propertyValues {Curve {keys [
          {arriveTangent=9.0725, interpMode=RCIM_Cubic, leaveTangent=9.0725, tangentMode=RCTM_User, time=0, value=0}
          {arriveTangentWeight=0.1274, interpMode=RCIM_Cubic, leaveTangentWeight=0.7778, tangentMode=RCTM_User, time=0.1408, value=0.9847}
          {arriveTangent=-1.7811, interpMode=RCIM_Cubic, leaveTangent=-1.7811, tangentMode=RCTM_User, time=1, value=0}
         ]}}})))
   &UpdateMeshOrientation(Rotation Rate=0.25, Rotation Vector=<0 0 1>)
   &ScaleColor(
    Scale Alpha=&Multiply_Float(
     A=&RandomRangeFloat(Minimum=0.2, Maximum=1)
     B=&Multiply_Float(
      A=&FloatFromCurve(FloatCurve=DataInterface() {propertyValues {Curve {keys [
           {interpMode=RCIM_Cubic, tangentMode=RCTM_User, time=0, value=0}
           {arriveTangent=-1, interpMode=RCIM_Cubic, leaveTangent=-1, time=0.2, value=0.95}
           {arriveTangent=-1.7637, interpMode=RCIM_Cubic, leaveTangent=-1.7637, tangentMode=RCTM_User, time=1, value=0}
          ]}}})
      B=&FNTimeOfDayScalar(Night Scale=0.8, Morning Scale=0.3, Evening Scale=0.5)
     )
    )
    Scale RGB=&VectorFromFloat(Value=&RandomRangeFloat(Minimum=5, Maximum=10))
    ScaleA=true
    ScaleRGB=true
   )
  ]
  rendererClasses [&NiagaraMeshRendererProperties]
  renderers [{rendererClass=&NiagaraMeshRendererProperties, rendererIndex=0}]
 }
 Fragment_MainWings_R {
  simTarget=GPUComputeSim
  emitterUpdateScript [
   &EmitterState()
   &SpawnRate(
    SpawnRate=&Multiply_Float(
     A=50
     B=&MakeCustomFloatFromBool(
      Bool=&Boolean_AndOperation(
       Bool1=$User.bisFullyDeployed
       Bool2=&SetBoolByFloatComparison(A=$User.ForwardDot, B=0.1, Comparison Type=::NewEnumerator6)
      )
      True Float=1
      False Float=0
     )
    )
    Use Spawn Probability=false
   )
  ]
  particleSpawnScript [
   &InitializeParticle(
    Color=&MakeCustomLinearColorFromBool(Bool=&RandomBool(), True Color=<rgba 0.9083 0.1374 1.5 1>, False Color=<rgba 0.1213 0.3179 1.5 1>)
    Lifetime Max=5
    Lifetime Min=0.5
    Lifetime Mode=::NewEnumerator1
    Mesh Scale Mode=::NewEnumerator1
    Mesh Uniform Scale Max=0.75
    Mesh Uniform Scale Min=0.25
    UsePositionOffset=false
    Write Material Random=true
   )
   &InitialMeshOrientation(
    Rotation=&RandomRangeVector(Minimum=<0 0 0>, Maximum=<0 0 1>)
    Use Orientation Vector=false
    Use Rotation Vector=true
   )
   &CylinderLocation(
    Cylinder Height=0
    Cylinder Radius=75
    Hemicircle Y=true
    MaskSpawnGroup=true
    Offset=<15 15 18>
    Offset Mode:ENiagara_OffsetMode=::NewEnumerator0
    Override Seed=false
    Roll=-0.035
    Rotation Mode:ENiagara_RotationMode=::NewEnumerator2
    UseOwnerScale=true
    Yaw=-0.225
   )
   &KillParticlesInVolume(
    Cone Angle=30
    Cone Axis:Vector3f=<-1 -0.25 0>
    Cone Length=500
    Invert Volume=true
    Kill Shape=::NewEnumerator4
    Origin Offset:Vector3f=<0 0 20>
   )
   &CameraOffset(Camera Offset Amount=&RandomRangeFloat(Minimum=-15, Maximum=5))
  ]
  particleUpdateScript [
   &ParticleState()
   &PointForce(
    Force Falloff Distance=300
    Force Origin=&Add_Vector(A=&SimulationPosition(), B=<0 0 15>)
    Force Strength=200
    Linear Force Vector:Vector3f=<0 0 0>
    Module.Use Falloff=true
    Module.UseLinearForce=false
   )
   &CurlNoiseForce(Pan Noise=false)
   &AccelerationForce(Acceleration=<0 0 -10>)
   &Drag(Drag=3, Use Linear Drag=true, Use Rotational Drag=false)
   &SpringForce(
    " Particle Equilibrium Position":NiagaraPosition=&Add_Vector(A=&SimulationPosition(), B=<15 15 18>)
    Force Strength=0.05
    Spring Tightness=3
   )
   &SolveForcesAndVelocity(Clamp Velocity=false, Limit Acceleration=false)
   &ScaleMeshSize(Scale Factor=&VectorFromFloat(Value=&FloatFromCurve(FloatCurve=DataInterface() {propertyValues {Curve {keys [
          {arriveTangent=9.0725, interpMode=RCIM_Cubic, leaveTangent=9.0725, tangentMode=RCTM_User, time=0, value=0}
          {arriveTangentWeight=0.1274, interpMode=RCIM_Cubic, leaveTangentWeight=0.7778, tangentMode=RCTM_User, time=0.1408, value=0.9847}
          {arriveTangent=-1.7811, interpMode=RCIM_Cubic, leaveTangent=-1.7811, tangentMode=RCTM_User, time=1, value=0}
         ]}}})))
   &UpdateMeshOrientation(Rotation Rate=0.25, Rotation Vector=<0 0 1>)
   &ScaleColor(
    Scale Alpha=&Multiply_Float(
     A=&RandomRangeFloat(Minimum=0.2, Maximum=1)
     B=&Multiply_Float(
      A=&FloatFromCurve(FloatCurve=DataInterface() {propertyValues {Curve {keys [
           {interpMode=RCIM_Cubic, tangentMode=RCTM_User, time=0, value=0}
           {arriveTangent=-1, interpMode=RCIM_Cubic, leaveTangent=-1, time=0.2, value=0.95}
           {arriveTangent=-1.7637, interpMode=RCIM_Cubic, leaveTangent=-1.7637, tangentMode=RCTM_User, time=1, value=0}
          ]}}})
      B=&FNTimeOfDayScalar(Night Scale=0.8, Morning Scale=0.3, Evening Scale=0.5)
     )
    )
    Scale RGB=&VectorFromFloat(Value=&RandomRangeFloat(Minimum=5, Maximum=10))
    ScaleA=true
    ScaleRGB=true
   )
  ]
  rendererClasses [&NiagaraMeshRendererProperties]
  renderers [{rendererClass=&NiagaraMeshRendererProperties, rendererIndex=0}]
 }
 Fragment_TailFeathers {
  simTarget=GPUComputeSim
  emitterUpdateScript [
   &EmitterState()
   &SpawnRate(
    SpawnRate=&Multiply_Float(
     A=10
     B=&MakeCustomFloatFromBool(
      Bool=&Boolean_AndOperation(
       Bool1=$User.bisFullyDeployed
       Bool2=&SetBoolByFloatComparison(A=$User.ForwardDot, B=0.1, Comparison Type=::NewEnumerator6)
      )
      True Float=1
      False Float=0
     )
    )
    Use Spawn Probability=true
   )
  ]
  particleSpawnScript [
   &InitializeParticle(
    Color=&MakeCustomLinearColorFromBool(Bool=&RandomBool(), True Color=<rgba 0.9083 0.1374 1.5 1>, False Color=<rgba 0.1213 0.3179 1.5 1>)
    Lifetime Max=5
    Lifetime Min=0.5
    Lifetime Mode=::NewEnumerator1
    Mesh Scale Mode=::NewEnumerator1
    Mesh Uniform Scale Max=0.75
    Mesh Uniform Scale Min=0.25
    UsePositionOffset=false
    Write Material Random=true
   )
   &InitialMeshOrientation(
    Rotation=&RandomRangeVector(Minimum=<0 0 0>, Maximum=<0 0 1>)
    Use Orientation Vector=false
    Use Rotation Vector=true
   )
   &CylinderLocation(
    Cylinder Height=0
    Cylinder Radius=25
    Hemicircle Y=true
    MaskSpawnGroup=true
    Offset=<0 20 20>
    Offset Mode:ENiagara_OffsetMode=::NewEnumerator0
    Override Seed=false
    UseOwnerScale=true
   )
   &KillParticlesInVolume(Cone Axis:Vector3f=<0 -1 0>, Cone Length=500, Kill Shape=::NewEnumerator4, Origin Offset:Vector3f=<0 0 18>)
   &CameraOffset(Camera Offset Amount=&RandomRangeFloat(Minimum=-15, Maximum=5))
  ]
  particleUpdateScript [
   &ParticleState()
   &PointForce(
    Force Origin=&Add_Vector(A=&SimulationPosition(), B=<0 0 15>)
    Linear Force Vector:Vector3f=<0 0 0>
    Module.Use Falloff=true
    Module.UseLinearForce=false
   )
   &CurlNoiseForce(Pan Noise=false)
   &Drag(Drag=1.5, Use Linear Drag=true, Use Rotational Drag=false)
   &SpringForce(
    " Particle Equilibrium Position":NiagaraPosition=&Add_Vector(A=&SimulationPosition(), B=<0 20 20>)
    Force Strength=0.075
    Spring Tightness=5
   )
   &SolveForcesAndVelocity(Clamp Velocity=false, Limit Acceleration=false)
   &ScaleMeshSize(Scale Factor=&VectorFromFloat(Value=&FloatFromCurve(FloatCurve=DataInterface() {propertyValues {Curve {keys [
          {arriveTangent=9.0725, interpMode=RCIM_Cubic, leaveTangent=9.0725, tangentMode=RCTM_User, time=0, value=0}
          {arriveTangentWeight=0.1274, interpMode=RCIM_Cubic, leaveTangentWeight=0.7778, tangentMode=RCTM_User, time=0.15, value=1}
          {arriveTangent=-2.0558, interpMode=RCIM_Cubic, leaveTangent=-2.0558, tangentMode=RCTM_User, time=1, value=0}
         ]}}})))
   &UpdateMeshOrientation(Rotation Rate=0.25, Rotation Vector=<0 0 1>)
   &ScaleColor(
    Scale Alpha=&Multiply_Float(
     A=&RandomRangeFloat(Minimum=0.2, Maximum=1)
     B=&Multiply_Float(
      A=&FloatFromCurve(FloatCurve=DataInterface() {propertyValues {Curve {keys [
           {interpMode=RCIM_Cubic, tangentMode=RCTM_User, time=0, value=0}
           {arriveTangent=-1, interpMode=RCIM_Cubic, leaveTangent=-1, time=0.2, value=0.95}
           {arriveTangent=-1.7637, interpMode=RCIM_Cubic, leaveTangent=-1.7637, tangentMode=RCTM_User, time=1, value=0}
          ]}}})
      B=&FNTimeOfDayScalar(Night Scale=0.8, Morning Scale=0.3, Evening Scale=0.5)
     )
    )
    Scale RGB=&VectorFromFloat(Value=&RandomRangeFloat(Minimum=5, Maximum=10))
    ScaleA=true
    ScaleRGB=true
   )
  ]
  rendererClasses [&NiagaraMeshRendererProperties]
  renderers [{rendererClass=&NiagaraMeshRendererProperties, rendererIndex=0}]
 }
 Fragment_TailFeathers001 {
  simTarget=GPUComputeSim
  emitterUpdateScript [
   &EmitterState()
   &SpawnRate(
    SpawnRate=&Multiply_Float(
     A=10
     B=&MakeCustomFloatFromBool(
      Bool=&Boolean_AndOperation(
       Bool1=$User.bisFullyDeployed
       Bool2=&SetBoolByFloatComparison(A=$User.ForwardDot, B=0.1, Comparison Type=::NewEnumerator6)
      )
      True Float=1
      False Float=0
     )
    )
    Use Spawn Probability=false
   )
  ]
  particleSpawnScript [
   &InitializeParticle(
    Color=&MakeCustomLinearColorFromBool(Bool=&RandomBool(), True Color=<rgba 0.9083 0.1374 1.5 1>, False Color=<rgba 0.1213 0.3179 1.5 1>)
    Lifetime Max=5
    Lifetime Min=0.5
    Lifetime Mode=::NewEnumerator1
    Mesh Scale Mode=::NewEnumerator1
    Mesh Uniform Scale Max=0.75
    Mesh Uniform Scale Min=0.25
    UsePositionOffset=false
    Write Material Random=true
   )
   &InitialMeshOrientation(
    Rotation=&RandomRangeVector(Minimum=<0 0 0>, Maximum=<0 0 1>)
    Use Orientation Vector=false
    Use Rotation Vector=true
   )
   &CylinderLocation(
    Cylinder Height=0
    Cylinder Radius=25
    Hemicircle Y=true
    MaskSpawnGroup=true
    Non Uniform Scale:Vector3f=<1 -1 1>
    Offset=<0 8 20>
    Offset Mode:ENiagara_OffsetMode=::NewEnumerator0
    Override Seed=false
    UseOwnerScale=true
   )
   &KillParticlesInVolume(Cone Angle=70, Cone Axis:Vector3f=<0 1 0>, Cone Length=500, Kill Shape=::NewEnumerator4, Origin Offset:Vector3f=<0 20 20>)
   &CameraOffset(Camera Offset Amount=&RandomRangeFloat(Minimum=-15, Maximum=5))
  ]
  particleUpdateScript [
   &ParticleState()
   &PointForce(
    Force Falloff Distance=150
    Force Origin=&Add_Vector(A=&SimulationPosition(), B=<0 15 15>)
    Linear Force Vector:Vector3f=<0 0 0>
    Module.Use Falloff=true
    Module.UseLinearForce=false
   )
   &CurlNoiseForce(Pan Noise=false)
   &Drag(Drag=1.5, Use Linear Drag=true, Use Rotational Drag=false)
   &SpringForce(
    " Particle Equilibrium Position":NiagaraPosition=&Add_Vector(A=&SimulationPosition(), B=<0 20 20>)
    Force Strength=0.075
    Spring Tightness=5
   )
   &SolveForcesAndVelocity(Clamp Velocity=false, Limit Acceleration=false)
   &ScaleMeshSize(Scale Factor=&VectorFromFloat(Value=&FloatFromCurve(FloatCurve=DataInterface() {propertyValues {Curve {keys [
          {arriveTangent=9.0725, interpMode=RCIM_Cubic, leaveTangent=9.0725, tangentMode=RCTM_User, time=0, value=0}
          {arriveTangentWeight=0.1274, interpMode=RCIM_Cubic, leaveTangentWeight=0.7778, tangentMode=RCTM_User, time=0.15, value=1}
          {arriveTangent=-2.0558, interpMode=RCIM_Cubic, leaveTangent=-2.0558, tangentMode=RCTM_User, time=1, value=0}
         ]}}})))
   &UpdateMeshOrientation(Rotation Rate=0.25, Rotation Vector=<0 0 1>)
   &ScaleColor(
    Scale Alpha=&Multiply_Float(
     A=&RandomRangeFloat(Minimum=0.2, Maximum=1)
     B=&Multiply_Float(
      A=&FloatFromCurve(FloatCurve=DataInterface() {propertyValues {Curve {keys [
           {interpMode=RCIM_Cubic, tangentMode=RCTM_User, time=0, value=0}
           {arriveTangent=-1, interpMode=RCIM_Cubic, leaveTangent=-1, time=0.2, value=0.95}
           {arriveTangent=-1.7637, interpMode=RCIM_Cubic, leaveTangent=-1.7637, tangentMode=RCTM_User, time=1, value=0}
          ]}}})
      B=&FNTimeOfDayScalar(Night Scale=0.8, Morning Scale=0.3, Evening Scale=0.5)
     )
    )
    Scale RGB=&VectorFromFloat(Value=&RandomRangeFloat(Minimum=5, Maximum=10))
    ScaleA=true
    ScaleRGB=true
   )
  ]
  rendererClasses [&NiagaraMeshRendererProperties]
  renderers [{rendererClass=&NiagaraMeshRendererProperties, rendererIndex=0}]
 }
 Fragment_World {
  simTarget=GPUComputeSim
  emitterUpdateScript [
   &EmitterState()
   &SpawnRate(
    SpawnRate=&Multiply_Float(A=2000, B=&MakeCustomFloatFromBool(Bool=$User.bisFullyDeployed, True Float=1, False Float=0))
    Use Spawn Probability=false
   ) {enabled=false}
   &SpawnPerUnit(
    Module.Use Max Movement Threshold=true
    Module.Use Movement Tolerance=false
    Spawn Probability:NiagaraFloat=&MakeCustomFloatFromBool(
     Bool=&Boolean_OrOperation(
      Bool1=&SetBoolByFloatComparison(A=$User.ForwardDot, B=0.1, Comparison Type=::NewEnumerator2)
      Bool2=&SetBoolByFloatComparison(A=$User.ForwardDot, B=-0.1, Comparison Type=::NewEnumerator7)
     )
     True Float=1
     False Float=0
    )
    Spawn Spacing=15
    Use Spawn Probability=true
   )
  ]
  particleSpawnScript [
   &InitializeParticle(
    Color=&MakeCustomLinearColorFromBool(Bool=&RandomBool(), True Color=<rgba 0.9083 0.1374 1.5 1>, False Color=<rgba 0.1213 0.3179 1.5 1>)
    Lifetime Max=1
    Lifetime Min=0.25
    Lifetime Mode=::NewEnumerator1
    Mesh Scale Mode=::NewEnumerator1
    Mesh Uniform Scale Max=0.75
    Mesh Uniform Scale Min=0.25
    UsePositionOffset=false
    Write Material Random=true
   )
   &InitialMeshOrientation(
    Rotation=&RandomRangeVector(Minimum=<0 0 0>, Maximum=<1 1 1>)
    Use Orientation Vector=false
    Use Rotation Vector=true
   )
   &CameraOffset(Camera Offset Amount=&RandomRangeFloat(Minimum=-15, Maximum=5))
   &SkeletalMeshLocation(MaskSpawnGroup=false, Mesh Sampling Type=::NewEnumerator5, OverrideRandomSeed=false, Skeletal Mesh=$User.SkeletalMesh)
   &SphereLocation(
    MaskSpawnGroup=false
    Override Seed=false
    Sphere Radius=20
    Surface Only=false
    UseNonUniformScale=false
    UseOffset=false
    UseSphereOrientation=false
   )
  ]
  particleUpdateScript [
   &ParticleState()
   &CurlNoiseForce(Pan Noise=false)
   &Drag(Drag=3, Use Linear Drag=true, Use Rotational Drag=false)
   &InheritSourceMovement(Applied Position Scale=<0.9 0.9 1>, ApplyForce=false, ApplyPosition=true, ApplyPositionFalloff=false)
   &SolveForcesAndVelocity(Clamp Velocity=false, Limit Acceleration=false)
   &ScaleMeshSize(Scale Factor=&VectorFromFloat(Value=&FloatFromCurve(FloatCurve=DataInterface() {propertyValues {Curve {keys [
          {arriveTangent=9.0725, interpMode=RCIM_Cubic, leaveTangent=9.0725, tangentMode=RCTM_User, time=0, value=0}
          {arriveTangentWeight=0.1274, interpMode=RCIM_Cubic, leaveTangentWeight=0.7778, tangentMode=RCTM_User, time=0.15, value=1}
          {arriveTangent=-2.0558, interpMode=RCIM_Cubic, leaveTangent=-2.0558, tangentMode=RCTM_User, time=1, value=0}
         ]}}})))
   &UpdateMeshOrientation(Rotation Rate=0.25, Rotation Vector=<0 0 1>)
   &ScaleColor(
    Scale Alpha=&Multiply_Float(
     A=&RandomRangeFloat(Minimum=0.2, Maximum=1)
     B=&Multiply_Float(
      A=&FloatFromCurve(FloatCurve=DataInterface() {propertyValues {Curve {keys [
           {interpMode=RCIM_Cubic, tangentMode=RCTM_User, time=0, value=0}
           {arriveTangent=-1, interpMode=RCIM_Cubic, leaveTangent=-1, time=0.2, value=0.95}
           {arriveTangent=-1.7637, interpMode=RCIM_Cubic, leaveTangent=-1.7637, tangentMode=RCTM_User, time=1, value=0}
          ]}}})
      B=&FNTimeOfDayScalar(Night Scale=0.8, Morning Scale=0.3, Evening Scale=0.5)
     )
    )
    Scale RGB=&VectorFromFloat(Value=&RandomRangeFloat(Minimum=5, Maximum=10))
    ScaleA=true
    ScaleRGB=true
   )
  ]
  rendererClasses [&NiagaraMeshRendererProperties]
  renderers [{rendererClass=&NiagaraMeshRendererProperties, rendererIndex=0}]
 }
 Light {
  emitterUpdateScript [
   &EmitterState(
    Inactive Response=::NewEnumerator1
    Life Cycle Mode=::NewEnumerator1
    Loop Behavior=::NewEnumerator1
    Loop Duration Mode=::NewEnumerator1
    UseLoopDelay=false
   )
   &SpawnBurst_Instantaneous(Use Spawn Probability=false, UseLoopCountLimit=false)
  ]
  particleSpawnScript [&InitializeParticle(Color Mode=::NewEnumerator0, Lifetime=0, UsePositionOffset=false, Write Material Random=true)]
  particleUpdateScript [
   &ParticleState(Kill Particles When Lifetime Has Elapsed=false, Let Infinitely Lived Particles Die When Emitter Deactivates=true)
   &FN_Light_Attributes(
    Color=<rgba 0 0.4833 1 1>
    Create Color Attribute=true
    Create Exponent Attribute=true
    Create Position Attribute=true
    Create Radius Attribute=true
    Exponent=5
    Position=<0 25 50>
    Radius=150
    Scale Color=&ClampFloat(
     Float=&Waveform("Waveform [1]"=::NewEnumerator0, "[1] Frequency"=1, "[1] Amplitude Min/Max"=<0 10>, "[2] Phase"=$Emitter.Age, "[3] Phase"=$Emitter.Age)
     Min=8
     Max=10
    )
   )
  ]
  rendererClasses [&NiagaraLightRendererProperties]
  renderers [{rendererClass=&NiagaraLightRendererProperties, rendererIndex=0}]
 }
 Locked {
  bEnabled=false
  emitterUpdateScript [
   &EmitterState(Life Cycle Mode=::NewEnumerator1, Loop Behavior=::NewEnumerator1, Loop Duration Mode=::NewEnumerator1, UseLoopDelay=false)
   &SampleSkeletalMesh(
    AdditionalYawPitchRoll=false
    Bone / Socket Index=1
    Mesh Sampling Type:ENiagaraSkelSamplingModeFull=::NewEnumerator5
    Skeletal Mesh:NiagaraDataInterfaceSkeletalMesh=$User.SkeletalMesh
    Socket Sampling Mode:ENiagaraSocketSamplingMode=::NewEnumerator5
   )
   &SampleSkeletalMesh(
    AdditionalYawPitchRoll=false
    Mesh Sampling Type:ENiagaraSkelSamplingModeFull=::NewEnumerator5
    Skeletal Mesh:NiagaraDataInterfaceSkeletalMesh=$User.SkeletalMesh
    Socket Sampling Mode:ENiagaraSocketSamplingMode=::NewEnumerator5
   ) {module=SampleSkeletalMesh001}
   &SpawnPerUnit(
    Max Movement Threshold=2000
    Module.Use Max Movement Threshold=true
    Module.Use Movement Tolerance=false
    Spawn Group=1
    Spawn Spacing=1
    Use Spawn Probability=false
    Velocity Vector:Vector3f=$Emitter.SampleSkeletalMesh.SampledVelocity
   ) {enabled=false}
   &SpawnPerUnit(
    Max Movement Threshold=2000
    Module.Use Max Movement Threshold=true
    Module.Use Movement Tolerance=false
    Spawn Spacing=15
    Use Spawn Probability=false
    Velocity Vector:Vector3f=$Emitter.SampleSkeletalMesh001.SampledVelocity
   ) {enabled=false, module=SpawnPerUnit001}
   &SpawnRate(SpawnRate=60, Use Spawn Probability=false)
   &SpawnRate(SpawnRate=60, Use Spawn Probability=false) {module=SpawnRate001}
  ]
  particleSpawnScript [
   &InitializeParticle(
    Mass Mode=::NewEnumerator3
    Sprite Rotation Mode=::NewEnumerator2
    Sprite Size Mode=::NewEnumerator3
    Uniform Sprite Size=7
    UsePositionOffset=false
    Write Material Random=true
   )
   /SetVariables(Particles.RibbonWidth=&VectorLength(Vector=&Subtract_Vector(A=$Emitter.SampleSkeletalMesh.SampledPosition, B=$Emitter.SampleSkeletalMesh001.SampledPosition))) {isSetParameters=true}
   &self:IsLatestParticleInSpawnGroup_01(Spawn Info 1:NiagaraSpawnInfo=$Emitter.SpawnRate.SpawnOutputInfo, Spawn Info 2:NiagaraSpawnInfo=$Emitter.SpawnRate001.SpawnOutputInfo) {module=IsLatestParticleInSpawnGroup_01}
  ]
  particleUpdateScript [
   &ParticleState()
   &SkeletalMeshLocation(
    Bone / Socket Index=&SpawnGroup()
    MaskSpawnGroup=false
    Mesh Sampling Type=::NewEnumerator5
    Orientation Sampling=::NewEnumerator1
    Skeletal Mesh=$User.SkeletalMesh
    Socket Sampling Mode=::NewEnumerator5
    Write To Particle Position:NiagaraBool=$Transient.FirstFrame
   ) {module=SkeletalMeshLocation001}
   &self:Pin_Latest_Particle_In_Spawn_Group_01(
    PinnedPos:NiagaraPosition=$Particles.SkeletalMeshLocation001.SampledPosition
    Spawn Info 1:NiagaraSpawnInfo=$Emitter.SpawnRate.SpawnOutputInfo
    Spawn Info 2:NiagaraSpawnInfo=$Emitter.SpawnRate001.SpawnOutputInfo
   ) {module=Pin_Latest_Particle_In_Spawn_Group_01}
   &SolveForcesAndVelocity(Clamp Velocity=false, Limit Acceleration=false)
   &self:RibbonFacing(Vector1:Vector3f=$Emitter.SampleSkeletalMesh001.SampledPosition, Vector2:Vector3f=$Emitter.SampleSkeletalMesh.SampledPosition) {module=RibbonFacing}
  ]
  rendererClasses [&NiagaraSpriteRendererProperties, &NiagaraRibbonRendererProperties]
  renderers [
   {rendererClass=&NiagaraSpriteRendererProperties}
   {rendererClass=&NiagaraRibbonRendererProperties, rendererIndex=1}
  ]
 }
 Rift_Meshs {
  emitterUpdateScript [
   &EmitterState(
    Inactive Response=::NewEnumerator1
    Life Cycle Mode=::NewEnumerator1
    Loop Behavior=::NewEnumerator1
    Loop Duration Mode=::NewEnumerator1
    UseLoopDelay=false
   )
   &SpawnBurst_Instantaneous(Spawn Count=2, Use Spawn Probability=false, UseLoopCountLimit=false)
  ]
  particleSpawnScript [
   &InitializeParticle(
    Color=<rgba 0.8088 1 0.9635 1>
    Lifetime=0
    Mesh Scale Mode=::NewEnumerator1
    Mesh Uniform Scale Max=1
    Mesh Uniform Scale Min=0.5
    Position Offset=<0 15 20>
    Sprite Rotation Angle=-0.025
    Sprite Rotation Mode=::NewEnumerator1
    Sprite Size Mode=::NewEnumerator3
    Uniform Sprite Size=25
    UsePositionOffset=true
    Write Material Random=true
   )
   &InitialMeshOrientation(Rotation=<0 0.25 0>, Use Orientation Vector=true, Use Rotation Vector=true)
   &DynamicMaterialParameters(
    Index 0 Param 1=1
    Index 0 Param 2=0.1
    Index 0 Param 3=1
    Index 0 Param 4=1
    Module.Param0WriteEnabled=true
    Module.Param1WriteEnabled=true
    Module.Param2WriteEnabled=true
    Module.Param3WriteEnabled=true
   )
   &CameraOffset(Camera Offset Amount=5)
  ]
  particleUpdateScript [&ParticleState(Kill Particles When Lifetime Has Elapsed=false, Let Infinitely Lived Particles Die When Emitter Deactivates=true)]
  rendererClasses [&NiagaraMeshRendererProperties]
  renderers [{rendererClass=&NiagaraMeshRendererProperties, rendererIndex=0}]
 }
 Round_Flare {
  emitterUpdateScript [
   &EmitterState()
   &SpawnRate(SpawnRate=2, Use Spawn Probability=false)
  ]
  particleSpawnScript [
   &InitializeParticle(
    Color=<rgba 0 0.4833 1 1>
    Lifetime=1.1
    Position Offset=<0 15 20>
    Sprite Size Mode=::NewEnumerator3
    Uniform Sprite Size=100
    UsePositionOffset=true
    Write Material Random=true
   )
   &CameraOffset(Camera Offset Amount=10)
  ]
  particleUpdateScript [
   &ParticleState()
   &ScaleColor(
    Scale Alpha=&FloatFromCurve(FloatCurve=DataInterface() {propertyValues {Curve {keys [
         {interpMode=RCIM_Cubic, tangentMode=RCTM_User, time=0, value=0}
         {interpMode=RCIM_Cubic, tangentMode=RCTM_User, time=0.5, value=1}
         {tangentMode=RCTM_User, time=1, value=0}
        ]}}})
    Scale RGB=<0.5 0.5 2>
    ScaleA=true
    ScaleRGB=true
   )
   &DynamicMaterialParameters(
    Index 0 Param 1=0.5
    Index 0 Param 2=0.5
    Index 0 Param 3=1
    Index 0 Param 4=1
    Module.Param0WriteEnabled=true
    Module.Param1WriteEnabled=true
    Module.Param2WriteEnabled=true
    Module.Param3WriteEnabled=true
   )
   &ScaleSpriteSize(
    Scale Sprite Size Mode=::NewEnumerator0
    Uniform Scale Factor=&RandomRangeFloat(Minimum=1, Maximum=1.5)
   )
   /SetVariables(Particles.Wide Flare Size:Vector2f=&Multiply_Vector2D_ByFloat(Float=&RandomRangeFloat(Minimum=0.75, Maximum=1.25), Vector2D=<250 5>)) {isSetParameters=true}
  ]
  rendererClasses [&NiagaraSpriteRendererProperties]
  renderers [
   {rendererClass=&NiagaraSpriteRendererProperties}
   {rendererClass=&NiagaraSpriteRendererProperties, rendererIndex=1}
  ]
 }
 Static_Sparkles {
  simTarget=GPUComputeSim
  emitterUpdateScript [
   &EmitterState()
   &SpawnRate(
    SpawnRate=&Multiply_Float(A=10, B=&MakeCustomFloatFromBool(Bool=$User.bisFullyDeployed, True Float=1, False Float=0))
    Use Spawn Probability=false
   )
  ]
  particleSpawnScript [
   &InitializeParticle(
    Color=<rgba 1.5469 2.6507 3 1>
    Lifetime Max=1.2
    Lifetime Min=0.3
    Lifetime Mode=::NewEnumerator1
    Sprite Rotation Mode=::NewEnumerator2
    Sprite Size Mode=::NewEnumerator1
    Uniform Sprite Size Max=24
    Uniform Sprite Size Min=12
    UsePositionOffset=false
    Write Material Random=true
   )
   &CameraOffset(Camera Offset Amount=15)
  ]
  particleUpdateScript [
   &ParticleState()
   &SkeletalMeshLocation(
    MaskSpawnGroup=false
    Mesh Sampling Type=::NewEnumerator7
    OffsetAlongNormal=false
    OverrideRandomSeed=false
    Skeletal Mesh=$User.SkeletalMesh
   )
   &DynamicMaterialParameters(
    Index 0 Param 1=&FloatFromCurve(FloatCurve=DataInterface() {propertyValues {Curve {keys [{arriveTangent=-3.2272, interpMode=RCIM_Cubic, leaveTangent=-3.2272, tangentMode=RCTM_User, time=0, value=1}, <at 1 0 RCIM_Cubic>]}}})
    Index 0 Param 2=&FloatFromCurve(FloatCurve=DataInterface() {propertyValues {Curve {keys [{arriveTangent=-12.4095, interpMode=RCIM_Cubic, leaveTangent=-12.4095, tangentMode=RCTM_User, time=0.5, value=2}, <at 1 0 RCIM_Cubic>]}}})
    Module.Param0WriteEnabled=true
    Module.Param1WriteEnabled=true
    Module.Param2WriteEnabled=true
    Module.Param3WriteEnabled=true
   )
   &ScaleSpriteSize(Uniform Curve Sprite Scale=DataInterface() {propertyValues {Curve {keys [
        {arriveTangent=14.9673, interpMode=RCIM_Cubic, leaveTangent=14.9673, tangentMode=RCTM_User, time=0, value=0}
        <at 0.2 1 RCIM_Cubic>
        {arriveTangentWeight=0.9046, interpMode=RCIM_Cubic, tangentMode=RCTM_User, time=1, value=0}
       ]}}})
   &ScaleColor(
    Scale RGB=&VectorFromFloat(Value=&FloatFromCurve(FloatCurve=DataInterface() {propertyValues {Curve {keys [
          {arriveTangent=321.9435, interpMode=RCIM_Cubic, leaveTangent=321.9435, tangentMode=RCTM_User, time=0, value=0}
          {arriveTangentWeight=1.8705, interpMode=RCIM_Cubic, leaveTangentWeight=1.8705, tangentMode=RCTM_User, time=0.05, value=5}
          <at 0.1 1 RCIM_Cubic>
         ]}}}))
    ScaleA=true
    ScaleRGB=true
   )
   &ScaleSpriteSize(
    Scale Sprite Size Mode=::NewEnumerator0
    Uniform Scale Factor=&FN_LOD_Distance_Remap(Start Distance=0, End Distance=2000, Minimum Output=1, Maximum Output=2)
   ) {enabled=false, module=ScaleSpriteSize001}
   &SolveForcesAndVelocity(Clamp Velocity=false, Limit Acceleration=false)
  ]
  rendererClasses [&NiagaraSpriteRendererProperties]
  renderers [{rendererClass=&NiagaraSpriteRendererProperties}]
 }
 Wyeth_AnimTrails {
  emitterUpdateScript [
   &EmitterState(Life Cycle Mode=::NewEnumerator1, Loop Duration=0.5, UseLoopDelay=false)
   &self:SpawnAnimTrail(
    AnimTrail is Active:NiagaraBool=&Boolean_AndOperation(
     Bool1=$User.bisFullyDeployed
     Bool2=&Boolean_OrOperation(
      Bool1=&SetBoolByFloatComparison(A=$User.ForwardDot, B=0.1, Comparison Type=::NewEnumerator2)
      Bool2=&SetBoolByFloatComparison(A=$User.ForwardDot, B=-0.1, Comparison Type=::NewEnumerator7)
     )
    )
    Mesh:NiagaraDataInterfaceSkeletalMesh=$User.SkeletalMesh
    Socket A=5
    Socket B=4
   ) {module=SpawnAnimTrail}
  ]
  particleSpawnScript [
   &self:InitializeAnimTrail(Ribbon Width Scale=1) {module=InitializeAnimTrail}
   &AddVelocity(Velocity=&MakeCustomVectorFromBool(Bool=$User.bisFrontEndPreview, True Vector=<0 0 300>, False Vector=<0 0 0>))
   &DynamicMaterialParameters(
    Index 0 Param 1=&RandomRangeFloat(Minimum=0, Maximum=1, Randomness Mode=::NewEnumerator1, Fixed Random Seed=true)
    Index 0 Param 2=&RandomRangeFloat(Minimum=0, Maximum=1, Randomness Mode=::NewEnumerator1, Random Seed=1, Fixed Random Seed=true)
    Index 0 Param 3=&RandomRangeFloat(Minimum=0, Maximum=1, Randomness Mode=::NewEnumerator1, Random Seed=2, Fixed Random Seed=true)
    Module.Param0WriteEnabled=true
    Module.Param1WriteEnabled=true
    Module.Param2WriteEnabled=true
    Module.Param3WriteEnabled=true
   )
  ]
  particleUpdateScript [
   &ParticleState() {module=UpdateAge}
   &SolveForcesAndVelocity(Clamp Velocity=false, Limit Acceleration=false)
   &self:UpdateAnimTrail() {module=UpdateAnimTrail}
   &ScaleColor(Scale Alpha=$System.Timeline.NormalizedTime, ScaleA=true, ScaleRGB=true) {module=ScaleColor001}
   &ScaleColor(
    Scale Alpha=&FloatFromCurve(
     FloatCurve=DataInterface() {propertyValues {Curve {keys [<at 0 1 RCIM_Cubic>, {arriveTangent=-2.1717, interpMode=RCIM_Cubic, leaveTangent=-2.1717, tangentMode=RCTM_User, time=1, value=0}]}}}
     Scale Curve=3
    )
    ScaleA=true
    ScaleRGB=true
   )
  ]
  rendererClasses [&NiagaraRibbonRendererProperties]
  renderers [{rendererClass=&NiagaraRibbonRendererProperties}]
 }
 Wyeth_AnimTrails001 {
  emitterUpdateScript [
   &EmitterState(Life Cycle Mode=::NewEnumerator1, Loop Duration=0.5, UseLoopDelay=false)
   &self:SpawnAnimTrail(
    AnimTrail is Active:NiagaraBool=&Boolean_AndOperation(
     Bool1=$User.bisFullyDeployed
     Bool2=&Boolean_OrOperation(
      Bool1=&SetBoolByFloatComparison(A=$User.ForwardDot, B=0.1, Comparison Type=::NewEnumerator2)
      Bool2=&SetBoolByFloatComparison(A=$User.ForwardDot, B=-0.1, Comparison Type=::NewEnumerator7)
     )
    )
    Mesh:NiagaraDataInterfaceSkeletalMesh=$User.SkeletalMesh
    Socket A=1
   ) {module=SpawnAnimTrail}
  ]
  particleSpawnScript [
   &self:InitializeAnimTrail(Ribbon Width Scale=1) {module=InitializeAnimTrail}
   &AddVelocity(Velocity=&MakeCustomVectorFromBool(Bool=$User.bisFrontEndPreview, True Vector=<0 0 300>, False Vector=<0 0 0>))
   &DynamicMaterialParameters(
    Index 0 Param 1=&RandomRangeFloat(Minimum=0, Maximum=1, Randomness Mode=::NewEnumerator1, Random Seed=3, Fixed Random Seed=true)
    Index 0 Param 2=&RandomRangeFloat(Minimum=0, Maximum=1, Randomness Mode=::NewEnumerator1, Random Seed=4, Fixed Random Seed=true)
    Index 0 Param 3=&RandomRangeFloat(Minimum=0, Maximum=1, Randomness Mode=::NewEnumerator1, Random Seed=5, Fixed Random Seed=true)
    Module.Param0WriteEnabled=true
    Module.Param1WriteEnabled=true
    Module.Param2WriteEnabled=true
    Module.Param3WriteEnabled=true
   )
  ]
  particleUpdateScript [
   &ParticleState() {module=UpdateAge}
   &SolveForcesAndVelocity(Clamp Velocity=false, Limit Acceleration=false)
   &self:UpdateAnimTrail() {module=UpdateAnimTrail}
   &ScaleColor(Scale Alpha=$System.Timeline.NormalizedTime, ScaleA=true, ScaleRGB=true) {module=ScaleColor001}
   &ScaleColor(
    Scale Alpha=&FloatFromCurve(
     FloatCurve=DataInterface() {propertyValues {Curve {keys [<at 0 1 RCIM_Cubic>, {arriveTangent=-2.1717, interpMode=RCIM_Cubic, leaveTangent=-2.1717, tangentMode=RCTM_User, time=1, value=0}]}}}
     Scale Curve=3
    )
    ScaleA=true
    ScaleRGB=true
   )
  ]
  rendererClasses [&NiagaraRibbonRendererProperties]
  renderers [{rendererClass=&NiagaraRibbonRendererProperties}]
 }
 Wyeth_AnimTrails002 {
  emitterUpdateScript [
   &EmitterState(Life Cycle Mode=::NewEnumerator1, Loop Duration=0.5, UseLoopDelay=false)
   &self:SpawnAnimTrail(
    AnimTrail is Active:NiagaraBool=&Boolean_AndOperation(
     Bool1=$User.bisFullyDeployed
     Bool2=&Boolean_OrOperation(
      Bool1=&SetBoolByFloatComparison(A=$User.ForwardDot, B=0.1, Comparison Type=::NewEnumerator2)
      Bool2=&SetBoolByFloatComparison(A=$User.ForwardDot, B=-0.1, Comparison Type=::NewEnumerator7)
     )
    )
    Mesh:NiagaraDataInterfaceSkeletalMesh=$User.SkeletalMesh
    Socket A=3
    Socket B=2
   ) {module=SpawnAnimTrail}
  ]
  particleSpawnScript [
   &self:InitializeAnimTrail(Ribbon Width Scale=1) {module=InitializeAnimTrail}
   &AddVelocity(Velocity=&MakeCustomVectorFromBool(Bool=$User.bisFrontEndPreview, True Vector=<0 0 300>, False Vector=<0 0 0>))
   &DynamicMaterialParameters(
    Index 0 Param 1=&RandomRangeFloat(Minimum=0, Maximum=1, Randomness Mode=::NewEnumerator1, Random Seed=6, Fixed Random Seed=true)
    Index 0 Param 2=&RandomRangeFloat(Minimum=0, Maximum=1, Randomness Mode=::NewEnumerator1, Random Seed=7, Fixed Random Seed=true)
    Index 0 Param 3=&RandomRangeFloat(Minimum=0, Maximum=1, Randomness Mode=::NewEnumerator1, Random Seed=8, Fixed Random Seed=true)
    Module.Param0WriteEnabled=true
    Module.Param1WriteEnabled=true
    Module.Param2WriteEnabled=true
    Module.Param3WriteEnabled=true
   )
  ]
  particleUpdateScript [
   &ParticleState() {module=UpdateAge}
   &SolveForcesAndVelocity(Clamp Velocity=false, Limit Acceleration=false)
   &self:UpdateAnimTrail() {module=UpdateAnimTrail}
   &ScaleColor(Scale Alpha=$System.Timeline.NormalizedTime, ScaleA=true, ScaleRGB=true) {module=ScaleColor001}
   &ScaleColor(
    Scale Alpha=&FloatFromCurve(
     FloatCurve=DataInterface() {propertyValues {Curve {keys [<at 0 1 RCIM_Cubic>, {arriveTangent=-2.1717, interpMode=RCIM_Cubic, leaveTangent=-2.1717, tangentMode=RCTM_User, time=1, value=0}]}}}
     Scale Curve=3
    )
    ScaleA=true
    ScaleRGB=true
   )
  ]
  rendererClasses [&NiagaraRibbonRendererProperties]
  renderers [{rendererClass=&NiagaraRibbonRendererProperties}]
 }
 Wyeth_AnimTrails003 {
  emitterUpdateScript [
   &EmitterState(Life Cycle Mode=::NewEnumerator1, Loop Duration=0.5, UseLoopDelay=false)
   &self:SpawnAnimTrail(
    AnimTrail is Active:NiagaraBool=&Boolean_AndOperation(
     Bool1=$User.bisFullyDeployed
     Bool2=&Boolean_OrOperation(
      Bool1=&SetBoolByFloatComparison(A=$User.ForwardDot, B=0.1, Comparison Type=::NewEnumerator2)
      Bool2=&SetBoolByFloatComparison(A=$User.ForwardDot, B=-0.1, Comparison Type=::NewEnumerator7)
     )
    )
    Mesh:NiagaraDataInterfaceSkeletalMesh=$User.SkeletalMesh
    Socket A=7
    Socket B=6
   ) {module=SpawnAnimTrail}
  ]
  particleSpawnScript [
   &self:InitializeAnimTrail(Ribbon Width Scale=1) {module=InitializeAnimTrail}
   &AddVelocity(Velocity=&MakeCustomVectorFromBool(Bool=$User.bisFrontEndPreview, True Vector=<0 0 300>, False Vector=<0 0 0>))
   &DynamicMaterialParameters(
    Index 0 Param 1=&RandomRangeFloat(Minimum=0, Maximum=1, Randomness Mode=::NewEnumerator1, Random Seed=9, Fixed Random Seed=true)
    Index 0 Param 2=&RandomRangeFloat(Minimum=0, Maximum=1, Randomness Mode=::NewEnumerator1, Random Seed=10, Fixed Random Seed=true)
    Index 0 Param 3=&RandomRangeFloat(Minimum=0, Maximum=1, Randomness Mode=::NewEnumerator1, Random Seed=11, Fixed Random Seed=true)
    Module.Param0WriteEnabled=true
    Module.Param1WriteEnabled=true
    Module.Param2WriteEnabled=true
    Module.Param3WriteEnabled=true
   )
  ]
  particleUpdateScript [
   &ParticleState() {module=UpdateAge}
   &SolveForcesAndVelocity(Clamp Velocity=false, Limit Acceleration=false)
   &self:UpdateAnimTrail() {module=UpdateAnimTrail}
   &ScaleColor(Scale Alpha=$System.Timeline.NormalizedTime, ScaleA=true, ScaleRGB=true) {module=ScaleColor001}
   &ScaleColor(
    Scale Alpha=&FloatFromCurve(
     FloatCurve=DataInterface() {propertyValues {Curve {keys [<at 0 1 RCIM_Cubic>, {arriveTangent=-2.1717, interpMode=RCIM_Cubic, leaveTangent=-2.1717, tangentMode=RCTM_User, time=1, value=0}]}}}
     Scale Curve=3
    )
    ScaleA=true
    ScaleRGB=true
   )
  ]
  rendererClasses [&NiagaraRibbonRendererProperties]
  renderers [{rendererClass=&NiagaraRibbonRendererProperties}]
 }
 Wyeth_AnimTrails004 {
  emitterUpdateScript [
   &EmitterState(Life Cycle Mode=::NewEnumerator1, Loop Duration=0.5, UseLoopDelay=false)
   &self:SpawnAnimTrail(
    AnimTrail is Active:NiagaraBool=&Boolean_AndOperation(
     Bool1=$User.bisFullyDeployed
     Bool2=&Boolean_OrOperation(
      Bool1=&SetBoolByFloatComparison(A=$User.ForwardDot, B=0.1, Comparison Type=::NewEnumerator2)
      Bool2=&SetBoolByFloatComparison(A=$User.ForwardDot, B=-0.1, Comparison Type=::NewEnumerator7)
     )
    )
    Mesh:NiagaraDataInterfaceSkeletalMesh=$User.SkeletalMesh
    Socket A=9
    Socket B=8
   ) {module=SpawnAnimTrail}
  ]
  particleSpawnScript [
   &self:InitializeAnimTrail(Ribbon Width Scale=1) {module=InitializeAnimTrail}
   &AddVelocity(Velocity=&MakeCustomVectorFromBool(Bool=$User.bisFrontEndPreview, True Vector=<0 0 300>, False Vector=<0 0 0>))
   &DynamicMaterialParameters(
    Index 0 Param 1=&RandomRangeFloat(Minimum=0, Maximum=1, Randomness Mode=::NewEnumerator1, Random Seed=12, Fixed Random Seed=true)
    Index 0 Param 2=&RandomRangeFloat(Minimum=0, Maximum=1, Randomness Mode=::NewEnumerator1, Random Seed=13, Fixed Random Seed=true)
    Index 0 Param 3=&RandomRangeFloat(Minimum=0, Maximum=1, Randomness Mode=::NewEnumerator1, Random Seed=14, Fixed Random Seed=true)
    Module.Param0WriteEnabled=true
    Module.Param1WriteEnabled=true
    Module.Param2WriteEnabled=true
    Module.Param3WriteEnabled=true
   )
  ]
  particleUpdateScript [
   &ParticleState() {module=UpdateAge}
   &SolveForcesAndVelocity(Clamp Velocity=false, Limit Acceleration=false)
   &self:UpdateAnimTrail() {module=UpdateAnimTrail}
   &ScaleColor(Scale Alpha=$System.Timeline.NormalizedTime, ScaleA=true, ScaleRGB=true) {module=ScaleColor001}
   &ScaleColor(
    Scale Alpha=&FloatFromCurve(
     FloatCurve=DataInterface() {propertyValues {Curve {keys [<at 0 1 RCIM_Cubic>, {arriveTangent=-2.1717, interpMode=RCIM_Cubic, leaveTangent=-2.1717, tangentMode=RCTM_User, time=1, value=0}]}}}
     Scale Curve=3
    )
    ScaleA=true
    ScaleRGB=true
   )
  ]
  rendererClasses [&NiagaraRibbonRendererProperties]
  renderers [{rendererClass=&NiagaraRibbonRendererProperties}]
 }
 Wyeth_AnimTrails005 {
  emitterUpdateScript [
   &EmitterState(Life Cycle Mode=::NewEnumerator1, Loop Duration=0.5, UseLoopDelay=false)
   &self:SpawnAnimTrail(
    AnimTrail is Active:NiagaraBool=&Boolean_AndOperation(
     Bool1=$User.bisFullyDeployed
     Bool2=&Boolean_OrOperation(
      Bool1=&SetBoolByFloatComparison(A=$User.ForwardDot, B=0.1, Comparison Type=::NewEnumerator2)
      Bool2=&SetBoolByFloatComparison(A=$User.ForwardDot, B=-0.1, Comparison Type=::NewEnumerator7)
     )
    )
    Mesh:NiagaraDataInterfaceSkeletalMesh=$User.SkeletalMesh
    Socket A=11
    Socket B=10
   ) {module=SpawnAnimTrail}
  ]
  particleSpawnScript [
   &self:InitializeAnimTrail(Ribbon Width Scale=1) {module=InitializeAnimTrail}
   &AddVelocity(Velocity=&MakeCustomVectorFromBool(Bool=$User.bisFrontEndPreview, True Vector=<0 0 300>, False Vector=<0 0 0>))
   &DynamicMaterialParameters(
    Index 0 Param 1=&RandomRangeFloat(Minimum=0, Maximum=1, Randomness Mode=::NewEnumerator1, Random Seed=15, Fixed Random Seed=true)
    Index 0 Param 2=&RandomRangeFloat(Minimum=0, Maximum=1, Randomness Mode=::NewEnumerator1, Random Seed=16, Fixed Random Seed=true)
    Index 0 Param 3=&RandomRangeFloat(Minimum=0, Maximum=1, Randomness Mode=::NewEnumerator1, Random Seed=17, Fixed Random Seed=true)
    Module.Param0WriteEnabled=true
    Module.Param1WriteEnabled=true
    Module.Param2WriteEnabled=true
    Module.Param3WriteEnabled=true
   )
  ]
  particleUpdateScript [
   &ParticleState() {module=UpdateAge}
   &SolveForcesAndVelocity(Clamp Velocity=false, Limit Acceleration=false)
   &self:UpdateAnimTrail() {module=UpdateAnimTrail}
   &ScaleColor(Scale Alpha=$System.Timeline.NormalizedTime, ScaleA=true, ScaleRGB=true) {module=ScaleColor001}
   &ScaleColor(
    Scale Alpha=&FloatFromCurve(
     FloatCurve=DataInterface() {propertyValues {Curve {keys [<at 0 1 RCIM_Cubic>, {arriveTangent=-2.1717, interpMode=RCIM_Cubic, leaveTangent=-2.1717, tangentMode=RCTM_User, time=1, value=0}]}}}
     Scale Curve=3
    )
    ScaleA=true
    ScaleRGB=true
   )
  ]
  rendererClasses [&NiagaraRibbonRendererProperties]
  renderers [{rendererClass=&NiagaraRibbonRendererProperties}]
 }
 Wyeth_AnimTrails006 {
  emitterUpdateScript [
   &EmitterState(Life Cycle Mode=::NewEnumerator1, Loop Duration=0.5, UseLoopDelay=false)
   &self:SpawnAnimTrail(
    AnimTrail is Active:NiagaraBool=&Boolean_AndOperation(
     Bool1=$User.bisFullyDeployed
     Bool2=&Boolean_OrOperation(
      Bool1=&SetBoolByFloatComparison(A=$User.ForwardDot, B=0.1, Comparison Type=::NewEnumerator2)
      Bool2=&SetBoolByFloatComparison(A=$User.ForwardDot, B=-0.1, Comparison Type=::NewEnumerator7)
     )
    )
    Mesh:NiagaraDataInterfaceSkeletalMesh=$User.SkeletalMesh
    Socket A=13
    Socket B=12
   ) {module=SpawnAnimTrail}
  ]
  particleSpawnScript [
   &self:InitializeAnimTrail(Ribbon Width Scale=1) {module=InitializeAnimTrail}
   &AddVelocity(Velocity=&MakeCustomVectorFromBool(Bool=$User.bisFrontEndPreview, True Vector=<0 0 300>, False Vector=<0 0 0>))
   &DynamicMaterialParameters(
    Index 0 Param 1=&RandomRangeFloat(Minimum=0, Maximum=1, Randomness Mode=::NewEnumerator1, Random Seed=18, Fixed Random Seed=true)
    Index 0 Param 2=&RandomRangeFloat(Minimum=0, Maximum=1, Randomness Mode=::NewEnumerator1, Random Seed=19, Fixed Random Seed=true)
    Index 0 Param 3=&RandomRangeFloat(Minimum=0, Maximum=1, Randomness Mode=::NewEnumerator1, Random Seed=20, Fixed Random Seed=true)
    Module.Param0WriteEnabled=true
    Module.Param1WriteEnabled=true
    Module.Param2WriteEnabled=true
    Module.Param3WriteEnabled=true
   )
  ]
  particleUpdateScript [
   &ParticleState() {module=UpdateAge}
   &SolveForcesAndVelocity(Clamp Velocity=false, Limit Acceleration=false)
   &self:UpdateAnimTrail() {module=UpdateAnimTrail}
   &ScaleColor(Scale Alpha=$System.Timeline.NormalizedTime, ScaleA=true, ScaleRGB=true) {module=ScaleColor001}
   &ScaleColor(
    Scale Alpha=&FloatFromCurve(
     FloatCurve=DataInterface() {propertyValues {Curve {keys [<at 0 1 RCIM_Cubic>, {arriveTangent=-2.1717, interpMode=RCIM_Cubic, leaveTangent=-2.1717, tangentMode=RCTM_User, time=1, value=0}]}}}
     Scale Curve=3
    )
    ScaleA=true
    ScaleRGB=true
   )
  ]
  rendererClasses [&NiagaraRibbonRendererProperties]
  renderers [{rendererClass=&NiagaraRibbonRendererProperties}]
 }
 Wyeth_AnimTrails007 {
  emitterUpdateScript [
   &EmitterState(Life Cycle Mode=::NewEnumerator1, Loop Duration=0.5, UseLoopDelay=false)
   &self:SpawnAnimTrail(
    AnimTrail is Active:NiagaraBool=&Boolean_AndOperation(
     Bool1=$User.bisFullyDeployed
     Bool2=&Boolean_OrOperation(
      Bool1=&SetBoolByFloatComparison(A=$User.ForwardDot, B=0.1, Comparison Type=::NewEnumerator2)
      Bool2=&SetBoolByFloatComparison(A=$User.ForwardDot, B=-0.1, Comparison Type=::NewEnumerator7)
     )
    )
    Mesh:NiagaraDataInterfaceSkeletalMesh=$User.SkeletalMesh
    Socket A=15
    Socket B=14
   ) {module=SpawnAnimTrail}
  ]
  particleSpawnScript [
   &self:InitializeAnimTrail(Ribbon Width Scale=1) {module=InitializeAnimTrail}
   &AddVelocity(Velocity=&MakeCustomVectorFromBool(Bool=$User.bisFrontEndPreview, True Vector=<0 0 300>, False Vector=<0 0 0>))
   &DynamicMaterialParameters(
    Index 0 Param 1=&RandomRangeFloat(Minimum=0, Maximum=1, Randomness Mode=::NewEnumerator1, Random Seed=21, Fixed Random Seed=true)
    Index 0 Param 2=&RandomRangeFloat(Minimum=0, Maximum=1, Randomness Mode=::NewEnumerator1, Random Seed=22, Fixed Random Seed=true)
    Index 0 Param 3=&RandomRangeFloat(Minimum=0, Maximum=1, Randomness Mode=::NewEnumerator1, Random Seed=23, Fixed Random Seed=true)
    Module.Param0WriteEnabled=true
    Module.Param1WriteEnabled=true
    Module.Param2WriteEnabled=true
    Module.Param3WriteEnabled=true
   )
  ]
  particleUpdateScript [
   &ParticleState() {module=UpdateAge}
   &SolveForcesAndVelocity(Clamp Velocity=false, Limit Acceleration=false)
   &self:UpdateAnimTrail() {module=UpdateAnimTrail}
   &ScaleColor(Scale Alpha=$System.Timeline.NormalizedTime, ScaleA=true, ScaleRGB=true) {module=ScaleColor001}
   &ScaleColor(
    Scale Alpha=&FloatFromCurve(
     FloatCurve=DataInterface() {propertyValues {Curve {keys [<at 0 1 RCIM_Cubic>, {arriveTangent=-2.1717, interpMode=RCIM_Cubic, leaveTangent=-2.1717, tangentMode=RCTM_User, time=1, value=0}]}}}
     Scale Curve=3
    )
    ScaleA=true
    ScaleRGB=true
   )
  ]
  rendererClasses [&NiagaraRibbonRendererProperties]
  renderers [{rendererClass=&NiagaraRibbonRendererProperties}]
 }
 Wyeth_AnimTrails008 {
  emitterUpdateScript [
   &EmitterState(Life Cycle Mode=::NewEnumerator1, Loop Duration=0.5, UseLoopDelay=false)
   &self:SpawnAnimTrail(
    AnimTrail is Active:NiagaraBool=&Boolean_AndOperation(
     Bool1=$User.bisFullyDeployed
     Bool2=&Boolean_OrOperation(
      Bool1=&SetBoolByFloatComparison(A=$User.ForwardDot, B=0.1, Comparison Type=::NewEnumerator2)
      Bool2=&SetBoolByFloatComparison(A=$User.ForwardDot, B=-0.1, Comparison Type=::NewEnumerator7)
     )
    )
    Mesh:NiagaraDataInterfaceSkeletalMesh=$User.SkeletalMesh
    Socket A=17
    Socket B=16
   ) {module=SpawnAnimTrail}
  ]
  particleSpawnScript [
   &self:InitializeAnimTrail(Ribbon Width Scale=1) {module=InitializeAnimTrail}
   &AddVelocity(Velocity=&MakeCustomVectorFromBool(Bool=$User.bisFrontEndPreview, True Vector=<0 0 300>, False Vector=<0 0 0>))
   &DynamicMaterialParameters(
    Index 0 Param 1=&RandomRangeFloat(Minimum=0, Maximum=1, Randomness Mode=::NewEnumerator1, Random Seed=24, Fixed Random Seed=true)
    Index 0 Param 2=&RandomRangeFloat(Minimum=0, Maximum=1, Randomness Mode=::NewEnumerator1, Random Seed=25, Fixed Random Seed=true)
    Index 0 Param 3=&RandomRangeFloat(Minimum=0, Maximum=1, Randomness Mode=::NewEnumerator1, Random Seed=26, Fixed Random Seed=true)
    Module.Param0WriteEnabled=true
    Module.Param1WriteEnabled=true
    Module.Param2WriteEnabled=true
    Module.Param3WriteEnabled=true
   )
  ]
  particleUpdateScript [
   &ParticleState() {module=UpdateAge}
   &SolveForcesAndVelocity(Clamp Velocity=false, Limit Acceleration=false)
   &self:UpdateAnimTrail() {module=UpdateAnimTrail}
   &ScaleColor(Scale Alpha=$System.Timeline.NormalizedTime, ScaleA=true, ScaleRGB=true) {module=ScaleColor001}
   &ScaleColor(
    Scale Alpha=&FloatFromCurve(
     FloatCurve=DataInterface() {propertyValues {Curve {keys [<at 0 1 RCIM_Cubic>, {arriveTangent=-2.1717, interpMode=RCIM_Cubic, leaveTangent=-2.1717, tangentMode=RCTM_User, time=1, value=0}]}}}
     Scale Curve=3
    )
    ScaleA=true
    ScaleRGB=true
   )
  ]
  rendererClasses [&NiagaraRibbonRendererProperties]
  renderers [{rendererClass=&NiagaraRibbonRendererProperties}]
 }
 Wyeth_AnimTrails009 {
  emitterUpdateScript [
   &EmitterState(Life Cycle Mode=::NewEnumerator1, Loop Duration=0.5, UseLoopDelay=false)
   &self:SpawnAnimTrail(
    AnimTrail is Active:NiagaraBool=&Boolean_AndOperation(
     Bool1=$User.bisFullyDeployed
     Bool2=&Boolean_OrOperation(
      Bool1=&SetBoolByFloatComparison(A=$User.ForwardDot, B=0.1, Comparison Type=::NewEnumerator2)
      Bool2=&SetBoolByFloatComparison(A=$User.ForwardDot, B=-0.1, Comparison Type=::NewEnumerator7)
     )
    )
    Mesh:NiagaraDataInterfaceSkeletalMesh=$User.SkeletalMesh
    Socket A=19
    Socket B=18
   ) {module=SpawnAnimTrail}
  ]
  particleSpawnScript [
   &self:InitializeAnimTrail(Ribbon Width Scale=1) {module=InitializeAnimTrail}
   &AddVelocity(Velocity=&MakeCustomVectorFromBool(Bool=$User.bisFrontEndPreview, True Vector=<0 0 300>, False Vector=<0 0 0>))
   &DynamicMaterialParameters(
    Index 0 Param 1=&RandomRangeFloat(Minimum=0, Maximum=1, Randomness Mode=::NewEnumerator1, Random Seed=27, Fixed Random Seed=true)
    Index 0 Param 2=&RandomRangeFloat(Minimum=0, Maximum=1, Randomness Mode=::NewEnumerator1, Random Seed=28, Fixed Random Seed=true)
    Index 0 Param 3=&RandomRangeFloat(Minimum=0, Maximum=1, Randomness Mode=::NewEnumerator1, Random Seed=29, Fixed Random Seed=true)
    Module.Param0WriteEnabled=true
    Module.Param1WriteEnabled=true
    Module.Param2WriteEnabled=true
    Module.Param3WriteEnabled=true
   )
  ]
  particleUpdateScript [
   &ParticleState() {module=UpdateAge}
   &SolveForcesAndVelocity(Clamp Velocity=false, Limit Acceleration=false)
   &self:UpdateAnimTrail() {module=UpdateAnimTrail}
   &ScaleColor(Scale Alpha=$System.Timeline.NormalizedTime, ScaleA=true, ScaleRGB=true) {module=ScaleColor001}
   &ScaleColor(
    Scale Alpha=&FloatFromCurve(
     FloatCurve=DataInterface() {propertyValues {Curve {keys [<at 0 1 RCIM_Cubic>, {arriveTangent=-2.1717, interpMode=RCIM_Cubic, leaveTangent=-2.1717, tangentMode=RCTM_User, time=1, value=0}]}}}
     Scale Curve=3
    )
    ScaleA=true
    ScaleRGB=true
   )
  ]
  rendererClasses [&NiagaraRibbonRendererProperties]
  renderers [{rendererClass=&NiagaraRibbonRendererProperties}]
 }
}
summary {
 systemName=NS_Glider_Buffet
 emitters [
  {emitterName=Fragment_TailFeathers, simTarget=GPUComputeSim, rendererClasses [&NiagaraMeshRendererProperties]}
  {emitterName=Fragment_MainWings_L, simTarget=GPUComputeSim, rendererClasses [&NiagaraMeshRendererProperties]}
  {emitterName=Fragment_MainWings_R, simTarget=GPUComputeSim, rendererClasses [&NiagaraMeshRendererProperties]}
  {emitterName=Fragment_TailFeathers001, simTarget=GPUComputeSim, rendererClasses [&NiagaraMeshRendererProperties]}
  {emitterName=Static_Sparkles, simTarget=GPUComputeSim, rendererClasses [&NiagaraSpriteRendererProperties]}
  {emitterName=FN_AnimTrail, rendererClasses [&NiagaraRibbonRendererProperties]}
  {emitterName=Caustic_Magic, rendererClasses [&NiagaraSpriteRendererProperties]}
  {emitterName=Fragment_World, simTarget=GPUComputeSim, rendererClasses [&NiagaraMeshRendererProperties]}
  {emitterName=Light, rendererClasses [&NiagaraLightRendererProperties]}
  {emitterName=Rift_Meshs, rendererClasses [&NiagaraMeshRendererProperties]}
  {emitterName=Round_Flare, rendererClasses [&NiagaraSpriteRendererProperties]}
  {emitterName=Wyeth_AnimTrails, rendererClasses [&NiagaraRibbonRendererProperties]}
  {emitterName=Wyeth_AnimTrails001, rendererClasses [&NiagaraRibbonRendererProperties]}
  {
   bEnabled=false
   emitterName=Locked
   rendererClasses [&NiagaraSpriteRendererProperties, &NiagaraRibbonRendererProperties]
  }
  {emitterName=Wyeth_AnimTrails002, rendererClasses [&NiagaraRibbonRendererProperties]}
  {emitterName=Wyeth_AnimTrails003, rendererClasses [&NiagaraRibbonRendererProperties]}
  {emitterName=Wyeth_AnimTrails004, rendererClasses [&NiagaraRibbonRendererProperties]}
  {emitterName=Wyeth_AnimTrails005, rendererClasses [&NiagaraRibbonRendererProperties]}
  {emitterName=Wyeth_AnimTrails006, rendererClasses [&NiagaraRibbonRendererProperties]}
  {emitterName=Wyeth_AnimTrails007, rendererClasses [&NiagaraRibbonRendererProperties]}
  {emitterName=Wyeth_AnimTrails008, rendererClasses [&NiagaraRibbonRendererProperties]}
  {emitterName=Wyeth_AnimTrails009, rendererClasses [&NiagaraRibbonRendererProperties]}
 ]
 userVariables [
  User.bisFrontEndPreview {type=NiagaraBool}
  User.bisFullyDeployed {type=NiagaraBool}
  User.ForwardDot {type=NiagaraFloat, defaultValue=0, description=""}
  User.RibbonIDOffset {type=NiagaraInt32}
  User.SkeletalMesh {
   type=NiagaraDataInterfaceSkeletalMesh
   defaultValue {
    struct=NiagaraExt_VariableValue_DataInterface
    value {dataInterface=&self:NiagaraDataInterfaceSkeletalMesh_1}
   }
  }
 ]
}
